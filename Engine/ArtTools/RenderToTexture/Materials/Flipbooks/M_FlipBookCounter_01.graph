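# Flipbook counter material in the RenderToTexture art tools that displays the Animation Phase value as a debug number.
# An Animation Phase scalar parameter feeds the DebugScalarValues material function, which renders the number.
# A feature level switch passes that result on the default path and substitutes a zero constant on ES2 and ES3_1.
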Material /Engine/ArtTools/RenderToTexture/Materials/Flipbooks/M_FlipBookCounter_01
graph {
 properties {pixelDepthOffsetMode=PDOM_Legacy, usedShadingModels=Default Lit}
 nodes [
  n1 = MaterialFunctionCall(Number=ScalarParameter() {parameterName=Animation Phase}) {materialFunction=&DebugScalarValues}
  n2 = Constant()
  FeatureLevelSwitch(Default=@n1.Result, ES2=@n2, ES3_1=@n2, SM4_REMOVED=@n1.Result)
 ]
}
