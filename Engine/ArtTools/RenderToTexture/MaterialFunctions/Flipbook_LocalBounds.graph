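# Material function that converts world position into local UVW coordinates relative to a Blueprint offset and frame size, also output as separate U, V and W.
# The offset position is divided by half the frame size, with an optional static-switch clamp to 0–1.
# An optional, costlier inverse transform supports a custom projection vector instead of top-down, but all three basis inputs read the same Vector2 output of CreateThirdOrthogonalVector.
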MaterialFunction /Engine/ArtTools/RenderToTexture/MaterialFunctions/Flipbook_LocalBounds
graph {nodes [
  n1 = Subtract(
   A=FunctionInput(Preview=WorldPosition().XYZ) {bUsePreviewValueAsDefault=true, inputName=World Position}
   B=FunctionInput() {bUsePreviewValueAsDefault=true, inputName=Blueprint Offset, sortPriority=1}
  )
  n2 = FunctionInput() {
   bUsePreviewValueAsDefault=true
   description="Only used if Transform Position is enabled. Defaults to 0,0,1 or from noon which is the cheapest. Only needed if projected is from the side. If you need a 90 degree rotation, use swizzle on world position rather than transform using a cardinal axis vector."
   inputName=Projection Vector
   sortPriority=6
   previewValue=<0.01 0 1 1>
  }
  n3 = MaterialFunctionCall!(Vector1=@n2, Vector2=Constant3Vector() {constant=<rgba 1 0 0 0>}) {materialFunction=/Engine/Functions/Engine_MaterialFunctions02/Math/CreateThirdOrthogonalVector}
  n4 = FunctionInput(Preview=StaticBool()) {
   bUsePreviewValueAsDefault=true
   description="When false (default), the projection will be from above (0,0,1) because that is cheaper than transforming into a custom vector. If necessary you can enable transform and supply a custom vector. If your custom vector is a cardinal axis like 1,0,0 or 0,1,0 consider swizzling worldposition which is cheaper than a transform."
   inputName=Transform Position
   inputType=FunctionInput_StaticBool
   sortPriority=5
   previewValue=<1 0 0 1>
  }
  n5 = Divide(
   A=StaticSwitch(
    True=MaterialFunctionCall!(VectorToTransform=@n1, BasisX=@n3.Vector2, BasisY=@n3.Vector2, BasisZ=@n3.Vector2) {materialFunction=/Engine/Functions/Engine_MaterialFunctions02/Math/InverseTransformMatrix}.Transformed Vector 3
    False=@n1
    Value=@n4
   )
   B=Multiply(A=FunctionInput() {inputName=Frame Size}) {constB=0.5}
  )
  n6 = FunctionInput(Preview=StaticBool()) {bUsePreviewValueAsDefault=true, inputName=Clamp, inputType=FunctionInput_StaticBool, sortPriority=2, previewValue=<1 0 0 1>}
  n7 = FunctionOutput(None=StaticSwitch(True=Clamp(None=@n5), False=@n5, Value=@n6)) {outputName=UVW}
  FunctionOutput(None=ComponentMask(None=@n7) {b=true}) {outputName=W, sortPriority=3}
  FunctionOutput(None=ComponentMask(None=@n7) {r=true}) {outputName=U, sortPriority=1}
  FunctionOutput(None=ComponentMask(None=@n7) {g=true}) {outputName=V, sortPriority=2}
 ]}
