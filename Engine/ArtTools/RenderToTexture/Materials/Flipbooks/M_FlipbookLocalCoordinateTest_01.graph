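# Unlit test material for the local-coordinate bounds function in the RenderToTexture flipbook art tools, judging by its name.
# Its graph holds one Flipbook_LocalBounds call, with a scalar Size parameter feeding Frame Size and the RGB of a Blueprint Offset vector parameter feeding that input.
# The function's other inputs are left unwired, and no connection from the call to a material output appears.
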Material /Engine/ArtTools/RenderToTexture/Materials/Flipbooks/M_FlipbookLocalCoordinateTest_01
graph {
 properties {pixelDepthOffsetMode=PDOM_Legacy, shadingModel=MSM_Unlit, usedShadingModels=Unlit}
 nodes [MaterialFunctionCall!(
   World Position
   Frame Size=ScalarParameter() {parameterName=Size}
   Blueprint Offset=VectorParameter() {parameterName=Blueprint Offset}.RGB
   Clamp
   Transform Position
   Projection Vector
  ) {materialFunction=/Engine/ArtTools/RenderToTexture/MaterialFunctions/Flipbook_LocalBounds}]
}
