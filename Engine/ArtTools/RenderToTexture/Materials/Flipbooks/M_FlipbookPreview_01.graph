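# Unlit masked material for previewing flipbook textures animated with motion vectors.
# FlipBook_MotionVectors samples the color and vector textures over a Rows by Columns grid, with playback blending between a manual frame Phase and speed-driven looping time via Use Speed.
# Static switches toggle custom UVs, double vector application, alpha masking, and a debug view splitting the forward vector's green component into red for positive and green for negative.
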Material /Engine/ArtTools/RenderToTexture/Materials/Flipbooks/M_FlipbookPreview_01
graph {
 properties {
  blendMode=BLEND_Masked
  numCustomizedUVs=3
  opacityMaskClipValue=0.5
  pixelDepthOffsetMode=PDOM_Legacy
  shadingModel=MSM_Unlit
  usedShadingModels=Unlit
 }
 nodes [
  MaterialFunctionCall() {materialFunction=&DebugScalarValues}
  n1 = ScalarParameter() {parameterName=Phase}
  n2 = ScalarParameter() {defaultValue=4, group=Flipbook Size, parameterName=Rows}
  n3 = ScalarParameter() {defaultValue=4, group=Flipbook Size, parameterName=Columns}
  n4 = Multiply(A=@n2, B=@n3)
  n5 = ScalarParameter() {parameterName=Speed}
  n6 = ScalarParameter() {defaultValue=1, parameterName=Use Speed}
  n7 = TextureObjectParameter() {parameterName=BaseColor, samplerType=SAMPLERTYPE_LinearColor, texture=/Engine/ArtTools/RenderToTexture/Textures/TilingTexturePreview}
  n8 = MaterialFunctionCall!(
   "Animation  Phase (0-1)"=LinearInterpolate(
    A=Divide(A=@n1, B=@n4)
    B=MaterialFunctionCall(Speed=@n5) {materialFunction=&TimeWithSpeedVariable}.Frac Time
    Alpha=Clamp(None=@n6)
   )
   Number of Columns X=@n3
   Number of Rows Y=@n2
   Texture=@n7
   UVs=TextureCoordinate()
   Motion Vector Texture=TextureObjectParameter() {parameterName=MotionVectors, samplerType=SAMPLERTYPE_LinearColor, texture=/Engine/ArtTools/RenderToTexture/Textures/127grey}
   Phase to Debug=ScalarParameter() {parameterName=DebugPhase}
   Motion Vector Intensity=ScalarParameter() {defaultValue=1, parameterName=MotionVec Intensity}
   Double Apply Motion Vectors=StaticBoolParameter() {parameterName=DoubleMotionVec}
   Custom UV Input=TextureCoordinate() {coordinateIndex=1}
   Custom UV2 Input=TextureCoordinate() {coordinateIndex=2}
   Use Custom UVs=StaticBoolParameter() {parameterName=CustomUVs}
  ) {materialFunction=/Engine/Functions/Engine_MaterialFunctions02/Texturing/FlipBook_MotionVectors}
  n9 = ComponentMask(None=ConstantBiasScale(None=@n8.Forward Motion Vectors) {bias=-0.5, scale=2}) {g=true}
  n10 = AppendVector(A=Max(A=@n9) {constB=0}, B=Max(A=Multiply(A=@n9) {constB=-1}) {constB=0})
  n11 = Constant()
  StaticSwitchParameter(True=Multiply(A=AppendVector(A=@n10, B=@n11)), False=@n8.Result) {parameterName=DebugVectors}
  StaticSwitchParameter(True=@n8.Alpha, False=Constant() {r=1}) {parameterName=Masked}
 ]
}
