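# Material function in the render-to-texture art tools that encodes flipbook motion vectors as the difference between two samples of a position texture.
# Texture parameter A is sampled at offset-frame and original-frame UVs, then the difference is scaled by RenderMotionVectorMultiply from MPC_R2T and remapped to 0-1.
# Pixels whose scaled X or Y magnitude exceeds 1 output solid red, flagging out-of-range vectors.
# Both sampled positions are also exposed as debug outputs.
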MaterialFunction /Engine/ArtTools/RenderToTexture/MaterialFunctions/DiffMotionVectors
graph {nodes [
  n1 = TextureCoordinate()
  n2 = TextureSampleParameter2D(UVs=FunctionInput(Preview=@n1) {bUsePreviewValueAsDefault=true, inputName=Offset Frame UVs, inputType=FunctionInput_Vector2, sortPriority=1}) {parameterName=A, samplerType=SAMPLERTYPE_LinearColor, texture=/Engine/ArtTools/RenderToTexture/Textures/FlipbookMotionVectors/Position1}
  n3 = TextureSampleParameter2D(UVs=FunctionInput(Preview=@n1) {bUsePreviewValueAsDefault=true, inputName=Original Frame UVs, inputType=FunctionInput_Vector2}) {parameterName=A, samplerType=SAMPLERTYPE_LinearColor, texture=&Position1}
  n4 = Multiply(
   A=Subtract(
    A=FunctionOutput(None=@n2.RGB) {description=World Position 1 - for debug, outputName=P2, sortPriority=2}
    B=FunctionOutput(None=@n3.RGB) {description="World Position 2 (frame offset) - for debug", outputName=P1, sortPriority=1}
   )
   B=CollectionParameter() {collection=/Engine/ArtTools/RenderToTexture/Materials/MPC_R2T, parameterName=RenderMotionVectorMultiply}
  )
  n5 = MaterialFunctionCall(Float3=Abs(None=@n4)) {materialFunction=&BreakOutFloat3Components}
  n6 = Multiply(A=Add(A=@n4)) {constB=0.5}
  n7 = If(A=@n5.G, "A > B"=Constant3Vector() {constant=<rgba 1 0 0 0>}, "A == B"=@n6, "A < B"=@n6) {constB=1}
  FunctionOutput(None=If(A=@n5.R, "A > B"=Constant3Vector() {constant=<rgba 1 0 0 0>}, "A == B"=@n7, "A < B"=@n7) {constB=1}) {description=The difference in motion vectors scaled into 0-1 range}
  Normalize(VectorInput=Multiply(A=@n4, B=Constant3Vector() {constant=<rgba 1 1 0 0>}))
 ]}
