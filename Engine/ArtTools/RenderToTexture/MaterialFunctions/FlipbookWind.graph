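# Material function for simple waving grass wind, meant as the last World Position Offset node applied.
# Time scrolls large and small world-position waves along a constant wind vector, shaped by smoothed triangle math rather than sines.
# Their sum drives RotateAboutAxis around the wind-cross-up axis, pivoting below the incoming AdditionalWPO.
# WindWeight and WindIntensity scale the result before adding AdditionalWPO; the Frame Size input is unconnected.
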MaterialFunction /Engine/ArtTools/RenderToTexture/MaterialFunctions/FlipbookWind
graph {
 properties {
  bExposeToLibrary=true
  description="Simple Waving grass wind, Grass must be textured to full sheet.  This should be the last WPO node applied."
  libraryCategoriesText [Misc, World Position Offset]
 }
 nodes [
  n1 = Constant4Vector() {desc=Wind, constant=<rgba 0 1 0 1>}
  n2 = Normalize(VectorInput=ComponentMask(None=@n1.RGBA) {b=true, g=true, r=true})
  n3 = FunctionInput(Preview=Multiply(A=Time()) {constB=0.285}) {bUsePreviewValueAsDefault=true, inputName=Time, inputType=FunctionInput_Scalar, previewValue=<1 0 0 1>}
  n4 = Multiply(A=ComponentMask(None=@n1.RGBA) {a=true}, B=@n3)
  n5 = Multiply(A=@n2, B=@n4)
  n6 = FunctionInput(Preview=WorldPosition().XYZ) {inputName=WorldPosition}
  n7 = FunctionInput() {bUsePreviewValueAsDefault=true, inputName=Wave Size Large 800, inputType=FunctionInput_Scalar, previewValue=<200 0 0 1>}
  n8 = Divide(A=@n6, B=@n7)
  n9 = Frac(None=Add(A=Add(A=@n5, B=@n8), B=Constant() {r=0.5}))
  n10 = Constant() {r=2}
  n11 = Abs(None=Add(A=Multiply(A=@n9, B=@n10), B=Constant() {r=-1})) {desc=custom1}
  n12 = Constant() {r=2}
  n13 = Multiply(A=Subtract(A=Constant() {r=3}, B=Multiply(A=@n11, B=@n12)), B=@n11)
  n14 = FunctionInput() {bUsePreviewValueAsDefault=true, inputName=Wave Size Small, inputType=FunctionInput_Scalar, previewValue=<1 0 0 1>}
  n15 = Divide(A=@n6, B=@n14)
  n16 = Frac(None=Add(A=Add(A=@n4, B=@n15), B=Constant() {r=0.5}))
  n17 = Constant() {r=2}
  n18 = Abs(None=Add(A=Multiply(A=@n16, B=@n17), B=Constant() {r=-1})) {desc=custom1}
  n19 = Constant() {r=2}
  n20 = Multiply(A=Subtract(A=Constant() {r=3}, B=Multiply(A=@n18, B=@n19)), B=@n18)
  n21 = Add(A=DotProduct(A=@n2, B=Multiply(A=@n13, B=@n11) {desc=custom2}), B=Distance(A=Multiply(A=@n20, B=@n18) {desc=custom2}, B=Constant()))
  AppendVector(B=@n21)
  Constant() {r=200}
  FunctionInput() {inputName=Frame Size}
  n22 = Constant3Vector() {constant=<rgba 0 0 1 0>}
  n23 = FunctionInput() {inputName=AdditionalWPO}
  n24 = Constant3Vector() {constant=<rgba 0 0 -10 0>}
  n25 = Multiply(
   A=RotateAboutAxis(NormalizedRotationAxis=CrossProduct(A=@n2, B=@n22), RotationAngle=@n21, PivotPoint=Add(A=@n23, B=@n24), Position=@n23)
   B=FunctionInput() {description=Input vertex colors or texturecoods, inputName=WindWeight, inputType=FunctionInput_Scalar}
  )
  n26 = FunctionInput() {description=How strong the effect should appear, inputName=WindIntensity, inputType=FunctionInput_Scalar}
  FunctionOutput(None=Add(A=Multiply(A=@n25, B=@n26), B=@n23))
  Multiply()
 ]
}
