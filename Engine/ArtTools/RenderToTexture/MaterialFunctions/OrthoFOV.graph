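# OrthoFOV is a material function that fakes perspective in orthographic renders by expanding foreground and shrinking background vertices.
# BoundingSphere_LightTransform computes local depth along a forward axis, remapped to a signed weight.
# That weight scales the vector from vertex to object position, multiplied by the tangent of the FOV amount, and adds to any incoming WPO.
# Forward axis and FOV default to the Render to Texture parameter collection.
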MaterialFunction /Engine/ArtTools/RenderToTexture/MaterialFunctions/OrthoFOV
graph {
 properties {
  bExposeToLibrary=true
  description=Uses Local Depth to expand foreground and shrink background verts around a center point. Can be used to give the appearance of FOV for orthographic views or renders.
  libraryCategoriesText [Misc, WorldPositionOffset]
 }
 nodes [
  n1 = FunctionInput() {bUsePreviewValueAsDefault=true, description=By default uses MPC value for Render to Texture blueprint., inputName=Additional WPO}
  n2 = ScalarParameter() {defaultValue=1, parameterName=Scale XY}
  n3 = StaticBool() {value=true}
  n4 = FunctionInput(Preview=ComponentMask(None=CollectionParameter() {collection=/Engine/ArtTools/RenderToTexture/Materials/MPC_R2T, parameterName=CameraForwardAxis}) {b=true, g=true, r=true}) {
   bUsePreviewValueAsDefault=true
   description=The world axis to use as the camera direction. By default uses MPC value from Render to Texture blueprint.
   inputName=Forward Axis
   previewValue=<1 0 0 1>
  }
  n5 = LinearInterpolate(Alpha=MaterialFunctionCall!(
    World Position=WorldPosition() {worldPositionShaderOffset=WPT_ExcludeAllShaderOffsets}.XYZ
    Scale XY=AppendVector(A=@n2, B=@n2)
    Scale Z=ScalarParameter() {defaultValue=1, parameterName=Scale Z}
    Clamp=@n3
    Transform to Local space=StaticBoolParameter() {parameterName=Transform to Local Space}
    Custom Light Vector=@n3
    Light Vector=@n4
   ) {materialFunction=&/BoundingSphere_LightTransform}.W) {constA=1, constB=-1}
  n6 = WorldPosition() {worldPositionShaderOffset=WPT_ExcludeAllShaderOffsets}
  MaterialExpressionFunctionInput_4 = FunctionInput(Preview=CollectionParameter() {collection=&MPC_R2T, parameterName=OrthoFOV}) {
   bUsePreviewValueAsDefault=true
   description=By default uses MPC value for Render to Texture blueprint.
   inputName=FOV Amount
   inputType=FunctionInput_Scalar
   sortPriority=1
   previewValue=<0.5 0 0 1>
  }
  n7 = Custom!(x=@MaterialExpressionFunctionInput_4) {
   code="return tan(x);"
   inputs [{inputName=x, input {expression=&self:MaterialExpressionFunctionInput_4}}]
  }
  n8 = LinearInterpolate(B=Add(A=Subtract(B=@n6.XYZ) {constA=0}, B=ObjectPositionWS()), Alpha=@n7)
  FunctionOutput(None=Add(A=@n1, B=Multiply(A=@n5, B=@n8))) {outputName=WorldPositionOffset}
 ]
}
