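# Material function, a special flipbook for the RenderToTexture blueprint that samples non-uniform flipbooks letterboxed inside uniform textures.
# Animation phase, optionally clamped before Frac, is multiplied by frame and row counts and floored to pick a grid cell.
# A column/row aspect factor feeds ScaleUVsByCenter and scales the cell offset, compensating for the letterbox.
# It samples with a mip bias, outputting RGB, alpha, UVs, blend phase and row index.
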MaterialFunction /Engine/Functions/Engine_MaterialFunctions02/Texturing/FlipBook_UniformNonUniform
graph {
 properties {description="This is a special version of the flipbook function only useful to the RenderToTexture blueprint. It allows you to sample Non-Uniform flipbooks from within Uniform Textures. Ie, like a letterboxed flipbook which is how all RenderTarget Previews are set up."}
 nodes [
  Constant2Vector() {g=0.5, r=1}
  n1 = FunctionInput(Preview=TextureCoordinate()) {
   bUsePreviewValueAsDefault=true
   description=Insert the texture coordinate of your choice.
   inputName=UVs
   inputType=FunctionInput_Vector2
   sortPriority=5
  }
  n2 = FunctionInput() {
   bUsePreviewValueAsDefault=true
   description=Enter the number of images horizontally across the texture.
   inputName=Number of Columns
   inputType=FunctionInput_Scalar
   previewValue=<2 0 0 1>
  }
  n3 = FunctionInput() {
   bUsePreviewValueAsDefault=true
   description=Enter the number of images vertically across the texture.
   inputName=Number of Rows
   inputType=FunctionInput_Scalar
   previewValue=<2 0 0 1>
  }
  n4 = AppendVector(A=@n2, B=@n3)
  n5 = AppendVector(A=Max(A=Divide(A=@n3, B=@n2), B=Constant() {r=1}), B=Max(A=Divide(A=@n2, B=@n3), B=Constant() {r=1}))
  n6 = FunctionInput(Preview=Time()) {
   bUsePreviewValueAsDefault=true
   description="Apply a \"frac\" material node to the input value if time is used to control the animation phase."
   inputName="Animation  Phase (0-1)"
   inputType=FunctionInput_Scalar
  }
  n7 = FunctionInput(Preview=StaticBool()) {
   bUsePreviewValueAsDefault=true
   description="A frac operation is applied to the animation phase. This causes a value of exactly 1 in the animation phase input to return 0. Enabling this feature will clamp the animation phase value from 0 to .9999 which avoids the issue. "
   inputName="Clamp Anim (see tooltip)"
   inputType=FunctionInput_StaticBool
   sortPriority=100
  }
  n8 = Frac(None=StaticSwitch(True=Clamp(None=@n6) {maxDefault=1}, False=@n6, Value=@n7)) {desc=0-1 animation}
  n9 = Multiply(A=AppendVector(A=Multiply(A=@n2, B=@n3), B=@n3), B=AppendVector(A=@n8, B=@n8))
  n10 = Floor(None=@n9)
  n11 = Divide(A=@n10, B=@n4)
  n12 = Add(
   A=MaterialFunctionCall(UVs=Divide(A=@n1, B=@n4), Texture Scale=@n5) {materialFunction=&ScaleUVsByCenter}.UVs
   B=Multiply(A=Frac(None=@n11), B=@n5)
  )
  n13 = TextureSample!(
   UVs=@n12
   Tex=FunctionInput(Preview=TextureObject() {texture=/Engine/Functions/Engine_MaterialFunctions02/ExampleContent/Textures/flipbook}) {bUsePreviewValueAsDefault=true, desc=Insert a texture object, inputName=Texture, inputType=FunctionInput_Texture2D}
   Bias=FunctionInput() {bUsePreviewValueAsDefault=true, inputName=MipBias, inputType=FunctionInput_Scalar, sortPriority=100}
   Apply View MipBias
  ) {mipValueMode=TMVM_MipBias}
  FunctionOutput(None=@n13.RGB)
  FunctionOutput(None=@n13.A) {outputName=Alpha, sortPriority=2}
  FunctionOutput(None=@n12) {outputName=UVs, sortPriority=1}
  FunctionOutput(None=Frac(None=ComponentMask(None=@n9) {r=true})) {outputName=Blend Phase, sortPriority=3}
  FunctionOutput(None=ComponentMask(None=@n10) {g=true}) {outputName=Y Row, sortPriority=1}
  MaterialFunctionCall!(In) {materialFunction=/Engine/Functions/Engine_MaterialFunctions02/Utility/Round}
 ]
}
