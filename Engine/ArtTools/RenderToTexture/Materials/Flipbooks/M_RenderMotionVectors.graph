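# Unlit masked render-to-texture material that renders motion vectors for flipbook and imposter textures.
# DiffMotionVectors compares flipbook UVs one frame apart, or neighboring X/Y cells with the Imposter switch, skipping last-row pitch.
# With MapMotionOntoFlipbook off, color instead encodes OrthoFOV-based position relative to FrameCenter.
# World position offset flattens the mesh into a selectable UV layout, with switches for debug output and frame-count drawing.
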Material /Engine/ArtTools/RenderToTexture/Materials/Flipbooks/M_RenderMotionVectors
graph {
 properties {
  bUsedWithStaticLighting=true
  blendMode=BLEND_Masked
  pixelDepthOffsetMode=PDOM_Legacy
  shadingModel=MSM_Unlit
  twoSided=true
  usedShadingModels=Unlit
 }
 nodes [
  n1 = TextureSampleParameter2D() {parameterName=Opacity Mask, texture=/Engine/EngineVolumetrics/Fogsheet/Materials/T_EV_BlankWhite_01}
  n2 = VectorParameter() {parameterName=Mask Channel, defaultValue=<rgba 0 0 0 1>}
  Add(
   A=DotProduct(A=AppendVector(A=@n1.RGB, B=@n1.RGB), B=AppendVector(A=@n2.RGB, B=@n2.RGB))
   B=ScalarParameter() {desc=Expand or shrink mask to tweak edge artifacts, parameterName=Mask Offset}
  )
  n3 = CollectionParameter() {collection=/Engine/ArtTools/RenderToTexture/Materials/MPC_R2T, parameterName=RenderMotionVectorMultiply}
  Clamp(None=@n3) {maxDefault=0, minDefault=0}
  n4 = ScalarParameter() {parameterName=Animation Phase}
  n5 = ScalarParameter() {defaultValue=8, parameterName=FramesY}
  n6 = ScalarParameter() {defaultValue=8, parameterName=FramesX}
  MaterialExpressionTextureCoordinate_4 = TextureCoordinate()
  MaterialExpressionTextureCoordinate_8 = TextureCoordinate() {coordinateIndex=1}
  MaterialExpressionTextureCoordinate_11 = TextureCoordinate() {coordinateIndex=2}
  MaterialExpressionTextureCoordinate_12 = TextureCoordinate() {coordinateIndex=3}
  MaterialExpressionScalarParameter_10 = ScalarParameter() {defaultValue=1, parameterName=UVIndex}
  n7 = Custom!(
   UV0=@MaterialExpressionTextureCoordinate_4
   UV1=@MaterialExpressionTextureCoordinate_8
   UV2=@MaterialExpressionTextureCoordinate_11
   UV3=@MaterialExpressionTextureCoordinate_12
   Index=@MaterialExpressionScalarParameter_10
  ) {
   code="float2 UVChoice[5];\r\nUVChoice[0]=UV0;\r\nUVChoice[1]=UV1;\r\nUVChoice[2]=UV2;\r\nUVChoice[3]=UV3;\r\nUVChoice[4]=UV3;\r\n\r\nreturn UVChoice[min((int)Index,4)];"
   outputType=CMOT_Float2
   inputs [
    {
     inputName=UV0
     input {expression=&self:MaterialExpressionTextureCoordinate_4}
    }
    {
     inputName=UV1
     input {expression=&self:MaterialExpressionTextureCoordinate_8}
    }
    {
     inputName=UV2
     input {expression=&self:MaterialExpressionTextureCoordinate_11}
    }
    {
     inputName=UV3
     input {expression=&self:MaterialExpressionTextureCoordinate_12}
    }
    {
     inputName=Index
     input {expression=&self:MaterialExpressionScalarParameter_10}
    }
   ]
  }
  n8 = Constant() {r=0.995}
  n9 = MaterialFunctionCall(UVs=@n7, Texture Scale=Divide(B=@n8)) {materialFunction=&ScaleUVsByCenter}
  n10 = MaterialFunctionCall!(
   "Animation  Phase (0-1)"=Subtract(A=@n4) {constB=0.001}
   Number of Rows=@n5
   Number of Columns=@n6
   Texture
   UVs=@n9.UVs
   MipBias
   "Clamp Anim (see tooltip)"
  ) {materialFunction=/Engine/Functions/Engine_MaterialFunctions02/Texturing/FlipBook_UniformNonUniform}
  n11 = Add(A=@n4, B=Divide(B=Multiply(A=@n5, B=@n6)))
  n12 = MaterialFunctionCall!(
   Original Frame UVs=@n10.UVs
   Offset Frame UVs=MaterialFunctionCall("Animation  Phase (0-1)"=Subtract(A=@n11) {constB=0.001}, Number of Rows=@n5, Number of Columns=@n6, UVs=@n9.UVs) {materialFunction=&FlipBook_UniformNonUniform}.UVs
  ) {materialFunction=/Engine/ArtTools/RenderToTexture/MaterialFunctions/DiffMotionVectors}
  ComponentMask(None=@n12.Result) {r=true}
  Constant4Vector() {constant=<rgba 1 0 0 0>}
  Constant() {r=4}
  Constant()
  n13 = Divide(B=@n6)
  n14 = Constant()
  n15 = MaterialFunctionCall(Original Frame UVs=@n10.UVs, Offset Frame UVs=Add(A=@n10.UVs, B=AppendVector(A=@n13, B=@n14)) {desc=Imposter X offset}) {desc=Z Rotation, materialFunction=&DiffMotionVectors}
  n16 = Add(A=@n10.Y Row)
  n17 = Constant() {r=0.5}
  n18 = AppendVector(A=@n14, B=Divide(B=@n5))
  n19 = ComponentMask(None=MaterialFunctionCall(Original Frame UVs=@n10.UVs, Offset Frame UVs=Add(A=@n10.UVs, B=@n18) {desc=Imposter Y offset}) {desc=Pitch, materialFunction=&DiffMotionVectors}.Result) {g=true}
  n20 = StaticSwitchParameter(
   True=AppendVector(
    A=ComponentMask(None=@n15.Result) {g=true, r=true}
    B=If(A=@n16, B=@n5, "A > B"=@n17, "A == B"=@n17, "A < B"=@n19) {desc=Skip last row Pitch}
   )
   False=@n12.Result
  ) {parameterName=Imposter}
  n21 = ConstantBiasScale(None=@n20) {bias=-0.5, scale=2}
  n22 = MaterialFunctionCall(Float3=@n21) {materialFunction=&BreakOutFloat3Components}
  n23 = Multiply(A=@n22.G) {constB=-1}
  n24 = AppendVector(A=AppendVector(A=Max(A=@n22.G) {constB=0}, B=Max(A=@n23) {constB=0}), B=Constant())
  n25 = ScalarParameter() {defaultValue=125, parameterName=Size}
  n26 = Multiply(A=Multiply(A=@n4, B=@n25), B=ScalarParameter() {parameterName=Debug_Phase*WPO Offset})
  n27 = WorldPosition() {worldPositionShaderOffset=WPT_ExcludeAllShaderOffsets}
  n28 = ScalarParameter() {parameterName=Animation Phase}
  n29 = Multiply(
   A=Multiply(A=Cosine(None=@n28), B=ScalarParameter() {defaultValue=40, parameterName=XOffset})
   B=Constant3Vector() {constant=<rgba -1 0 0 0>}
  )
  n30 = MaterialFunctionCall!(Additional WPO=Add(A=Add(A=@n26, B=@n27.XYZ), B=@n29), Forward Axis, FOV Amount) {materialFunction=/Engine/ArtTools/RenderToTexture/MaterialFunctions/OrthoFOV}
  n31 = VectorParameter() {parameterName=FrameCenter}
  n32 = Multiply(A=Divide(A=Subtract(A=@n30.WorldPositionOffset, B=@n31.RGB), B=@n25) {constB=1000}) {constB=0.5, desc=Negative values won't}
  n33 = StaticSwitch(
   True=StaticSwitchParameter(True=Multiply(A=@n24) {constB=5}, False=@n20) {parameterName=Debug G channel as Red and Green}
   False=Multiply(A=Add(A=@n32) {constB=0.5, desc=show up in render target}, B=Constant3Vector() {constant=<rgba 1 1 1 0>})
   Value=StaticBoolParameter() {parameterName=MapMotionOntoFlipbook}
  )
  n34 = Add(A=MaterialFunctionCall(Number=@n4) {materialFunction=&DebugScalarValues}.Result, B=@n33)
  n35 = StaticSwitchParameter(
   True=Constant()
   False=StaticSwitchParameter(True=Constant() {r=0.5}, False=StaticSwitchParameter(True=@n34, False=@n33) {parameterName=Draw Frame Count}) {parameterName=Debug 127grey}
  ) {parameterName=Black}
  n36 = Constant()
  FeatureLevelSwitch(Default=@n35, ES2=@n36, ES3_1=@n36, SM4_REMOVED=@n35)
  n37 = Subtract(B=WorldPosition().XYZ) {constA=0}
  n38 = ObjectPositionWS()
  n39 = Multiply(A=Divide(A=@n25), B=Multiply(A=@n8, B=ConstantBiasScale(None=@n7) {bias=-0.5, scale=2}))
  n40 = Constant()
  n41 = ScalarParameter() {parameterName=WPO Test}
  n42 = Multiply(A=@n21, B=Multiply(A=Divide(A=@n41, B=@n3), B=@n25))
  n43 = StaticSwitchParameter(True=Add(A=Add(A=@n37, B=@n38), B=AppendVector(A=@n39, B=@n40)), False=Add(A=@n29, B=@n42)) {defaultValue=true, parameterName=Unwrap UVs}
  n44 = Constant3Vector() {constant=<rgba 0 0 1 0>}
  FeatureLevelSwitch(Default=@n43, ES2=@n44, ES3_1=@n44, SM4_REMOVED=@n43)
  MaterialFunctionCall!(WindIntensity, WindWeight, Time, WorldPosition, Wave Size Small, Wave Size Large 800, Frame Size, AdditionalWPO) {materialFunction=/Engine/ArtTools/RenderToTexture/MaterialFunctions/FlipbookWind}
  MaterialFunctionCall(Additional WPO=@n42) {materialFunction=&OrthoFOV}
  Multiply(A=ConstantBiasScale() {bias=-0.5, scale=2})
  Multiply()
  Normalize(VectorInput=Multiply(A=@n12.Result, B=Constant3Vector() {constant=<rgba 1 1 0 0>}))
  ScalarParameter() {defaultValue=1, parameterName=MeshScale}
  TextureCoordinate()
 ]
}
